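; Placeable glow stick light actor with configurable color and brightness that breaks on taking damage.
; The construction script creates a dynamic material, sets one of three intensities, and applies the selected global palette color to point light, particles and emissive material.
; Server damage sets a replicated Destroy flag and destroys the actor.
; Its OnRep resets light color, stops particles, hides the mesh, plays an electric break sound and destroys the actor.
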(event EventOnDamageServer
  (params (Damage :type Float%20%28single-precision%29) (DamageTags :type Gameplay%20Tag%20Container%20Structure%20%28by%20ref%29) (Momentum :type Vector :default "0, 0, 0") (HitInfo :type Hit%20Result%20Structure%20%28by%20ref%29) (InstigatedBy :type Controller%20Object%20Reference) (DamageCauser :type Actor%20Object%20Reference) (EffectContext :type Gameplay%20Effect%20Context%20Handle%20Structure))
  (Variables|Default|SetDestroy true)
  (Actor|DestroyActor))

(fn OnRep_Destroy
  (Rendering|Components|Light|SetLightColor :self (Variables|Default|GetPointLight) :bSRGB true)
  (Components|Activation|Deactivate (Variables|Default|GetNS_GlowTorchAlt))
  (Development|SetHiddeninGame (Variables|Default|GetS_GlowStick) true false)
  (Audio|SpawnSoundAttached "/Game/Sounds/Fort_Traps/Transformer/ElectricitySpark_Break_SoundCue.ElectricitySpark_Break_SoundCue" (Variables|Default|GetS_GlowStick))
  (Actor|DestroyActor))

(fn SetLightBrightness
  (bind _pointlight (Variables|Default|GetPointLight))
  (switch Utilities|FlowControl|Switch|SwitchonEnum_LightBrightness (Variables|Light|GetLightBrightnessOptions)
    (:NewEnumerator0
      (Rendering|Components|Light|SetIntensity _pointlight 5.5))
    (:NewEnumerator1
      (Rendering|Components|Light|SetIntensity _pointlight 11))
    (:NewEnumerator2
      (Rendering|Components|Light|SetIntensity _pointlight 2.75))))

(fn SetLightColor
  (params (Color :type Linear%20Color%20Structure))
  (Rendering|Components|Light|SetLightColor (Variables|Default|GetPointLight) Color true)
  (Effects|Components|ParticleSystem|SetColorParameter (Variables|Default|GetNS_GlowTorchAlt) "color" Color)
  (Rendering|Material|SetVectorParameterValue (Variables|Default|GetGlowStick_Mat) "EmissiveMultiplier" Color :!class /Script/Engine.MaterialInstanceDynamic))

(fn ConstructionScript
  (bind _structlightcolorsglobal (Variables|Default|GetStructLightColorsGlobal))
  (bind (_blue_2_835dee5b4b1e8fe2dcbd2eb808f69637 _teal_4_32feb0134e88f33098310f87b27d1674 _purple_6_d175df4b447eed701e402daeeb0431e4 _yellow_8_47bf9e0c4852ef9111c4ce903b40a8b3 _white_10_2fdb792d4a2a1052ce1b4789ea94fe97 _orange_12_d80329cb4e7f729bbecd1b9b879bb103 _green_14_f93ebbc04e7685314d7338b74c858b52 _red_16_8f0f9b8040677bb0fb52e9bb1fd9a55e) (Utilities|Struct|BreakStructLightColorsGlobal _structlightcolorsglobal))
  (bind _returnvalue (Rendering|Material|CreateDynamicMaterialInstance (Variables|Default|GetS_GlowStick) 0 "/Game/Gadgets/Assets/GlowStick/Material/MI_FORT_Gadgets_Fragment_Container_Inst.MI_FORT_Gadgets_Fragment_Container_Inst" "None" :!class /Script/Engine.PrimitiveComponent))
  (Variables|Default|SetGlowStick_Mat _returnvalue)
  (CallFunction|SetLightBrightness)
  (switch Utilities|FlowControl|Switch|SwitchonEnum_LightColorsGlobal (Variables|Light|GetLightColorOptions)
    (:NewEnumerator7
      (CallFunction|SetLightColor _white_10_2fdb792d4a2a1052ce1b4789ea94fe97))
    (:NewEnumerator0
      (CallFunction|SetLightColor _blue_2_835dee5b4b1e8fe2dcbd2eb808f69637))
    (:NewEnumerator1
      (CallFunction|SetLightColor _teal_4_32feb0134e88f33098310f87b27d1674))
    (:NewEnumerator2
      (CallFunction|SetLightColor _purple_6_d175df4b447eed701e402daeeb0431e4))
    (:NewEnumerator3
      (CallFunction|SetLightColor _yellow_8_47bf9e0c4852ef9111c4ce903b40a8b3))
    (:NewEnumerator4
      (CallFunction|SetLightColor _orange_12_d80329cb4e7f729bbecd1b9b879bb103))
    (:NewEnumerator5
      (CallFunction|SetLightColor _green_14_f93ebbc04e7685314d7338b74c858b52))
    (:NewEnumerator6
      (CallFunction|SetLightColor _red_16_8f0f9b8040677bb0fb52e9bb1fd9a55e))))
